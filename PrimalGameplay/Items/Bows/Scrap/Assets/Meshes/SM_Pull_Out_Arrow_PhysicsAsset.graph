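# Physics asset for the pull-out arrow skeletal mesh of the Scrap bow item.
# It gives the arrow_nock and arrow_root bones one capsule body each.
# A single constraint joins the two bodies, with equal angular limits on both swing axes and twist.
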PhysicsAsset /PrimalGameplay/Items/Bows/Scrap/Assets/Meshes/SM_Pull_Out_Arrow_PhysicsAsset
bodies {
 arrow_nock [{shapeType=Capsule, length=24.2304, radius=3.9026, center=<23.9271 -0.1032 0.7169>, rotation=<rot 0 89.722 -89.9198>}]
 arrow_root [{shapeType=Capsule, length=22.2541, radius=4.6914, center=<55.0537 -0.0006 9.3225>, rotation=<rot 0 -89.975 -89.9077>}]
}
constraints [{bone1Name=arrow_nock, bone2Name=arrow_root, swing1LimitDegrees=45, swing2LimitDegrees=45, twistLimitDegrees=45}]
