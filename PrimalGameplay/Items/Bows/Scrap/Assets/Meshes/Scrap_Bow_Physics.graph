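# Physics asset for the Scrap Bow mesh that defines simulation bodies for three feather strands hanging from a string.
# A master capsule body anchors three two-segment chains, each segment a small capsule on its own bone.
# Constraints allow moderate swing at the master joint and tighter swing between segments, with twist nearly locked so the feathers sway without spinning.
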PhysicsAsset /PrimalGameplay/Items/Bows/Scrap/Assets/Meshes/Scrap_Bow_Physics
bodies {
 dyn_string_feather1_01 [{shapeType=Capsule, length=0.9355, radius=0.9527, center=<1.58 0 0>, rotation=<rot 90 0 0>}]
 dyn_string_feather1_02 [{shapeType=Capsule, length=4.0846, radius=0.5, center=<3.488 0 0>, rotation=<rot 90 0 0>}]
 dyn_string_feather2_01 [{shapeType=Capsule, length=0.9355, radius=0.9527, center=<1.58 0 0>, rotation=<rot 90 0 0>}]
 dyn_string_feather2_02 [{shapeType=Capsule, length=4.0846, radius=0.5, center=<3.0256 0 0.0001>, rotation=<rot 90 0 0>}]
 dyn_string_feather3_01 [{shapeType=Capsule, length=0.9355, radius=0.9527, center=<1.58 0 0>, rotation=<rot 90 0 0>}]
 dyn_string_feather3_02 [{shapeType=Capsule, length=4.001, radius=0.5836, center=<3.488 0 0>, rotation=<rot 90 0 0>}]
 dyn_string_feather_MSTR [{shapeType=Capsule, length=13.0346, radius=0.529, center=<2.7151 1.3184 1.9804>, rotation=<rot -54.2059 1.3949 20.2967>}]
}
constraints [
 {bone1Name=dyn_string_feather1_02, bone2Name=dyn_string_feather1_01, swing1LimitDegrees=20, swing2LimitDegrees=20, twistLimitDegrees=1}
 {bone1Name=dyn_string_feather2_02, bone2Name=dyn_string_feather2_01, swing1LimitDegrees=20, swing2LimitDegrees=20, twistLimitDegrees=1}
 {bone1Name=dyn_string_feather3_02, bone2Name=dyn_string_feather3_01, swing1LimitDegrees=20, swing2LimitDegrees=20, twistLimitDegrees=1}
 {bone1Name=dyn_string_feather1_01, bone2Name=dyn_string_feather_MSTR, swing1LimitDegrees=45, swing2LimitDegrees=45, twistLimitDegrees=2}
 {bone1Name=dyn_string_feather2_01, bone2Name=dyn_string_feather_MSTR, swing1LimitDegrees=45, swing2LimitDegrees=45, twistLimitDegrees=2}
 {bone1Name=dyn_string_feather3_01, bone2Name=dyn_string_feather_MSTR, swing1LimitDegrees=45, swing2LimitDegrees=45, twistLimitDegrees=2}
]
